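# Niagara swing effect for the one-handed female Exosuit pickaxe, combining layered weapon trails with a short particle burst.
# Three blue and one dark gray ribbon emitter use FN AnimTrail modules to follow a socket on User.Swing_SK_Mesh, each with its own width and alpha fade.
# DirectionalBurst spawns short-lived, fading particles with sprite and mesh renderers, scaling sprites by speed.
# Its skeletal mesh location module is disabled.
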NiagaraSystem /BRCosmetics/Weapons/FORT_Melee/Pickaxe_Exosuit_Female_1h/FX/NS_Pickaxe_Exosuit_Female_Swing
dependencies {
 usedDataInterfaces [&NiagaraDataInterfaceCurve, &NiagaraDataInterfaceColorCurve]
 usedDynamicInputs []
 usedModules [&SystemState]
}
emitters {
 DirectionalBurst {
  emitterUpdateScript [
   &EmitterState(Life Cycle Mode=::NewEnumerator1, Loop Behavior=::NewEnumerator1, Loop Duration=1, UseLoopDelay=false)
   &SpawnBurst_Instantaneous(Spawn Count=75, Use Spawn Probability=false, UseLoopCountLimit=false)
  ]
  particleSpawnScript [&InitializeParticle(
    Lifetime=0.3
    Mass Max=2
    Mass Mode=::NewEnumerator3
    Sprite Size Max=<3 14>
    Sprite Size Min=<2 6>
    Sprite Size Mode=::NewEnumerator2
    UsePositionOffset=false
    Write Material Random=true
   )]
  particleUpdateScript [
   &ParticleState()
   &SkeletalMeshLocation(Bone Sampling Mode=::NewEnumerator4, MaskSpawnGroup=false, Skeletal Mesh=$User.Swing_SK_Mesh) {enabled=false}
   &ScaleColor(
    Scale Alpha=&FloatFromCurve(FloatCurve=DataInterface() {propertyValues {Curve {keys [<at -0.0046 1.0113 RCIM_Cubic>,<at 1.0011 0.002 RCIM_Cubic>]}}})
    Scale RGB=&VectorFromFloat(Value=&FloatFromCurve(FloatCurve=DataInterface() {propertyValues {Curve {keys [<at -0.0046 1.0113 RCIM_Cubic>,<at 1.0011 0.002 RCIM_Cubic>]}}}, Scale Curve=0.1))
    ScaleA=true
    ScaleRGB=true
   )
   &SolveForcesAndVelocity(Clamp Velocity=false, Limit Acceleration=false)
   &ScaleSpriteSizeBySpeed(
    Max Scale Factor=<1.2 3.3>
    Min Scale Factor=<0.75 0.3>
    Sample Scale Factor By Curve=true
    Scale Factor Curve=DataInterface() {propertyValues {Curve {keys [
        {arriveTangent=2.865, interpMode=RCIM_Cubic, time=0.1247, value=-0.0006}
        {arriveTangent=4.1522, interpMode=RCIM_Cubic, leaveTangent=4.1522, tangentMode=RCTM_User, time=0.3948, value=0.4893}
        <at 0.9958 0.9974 RCIM_Cubic>
       ]}}}
    Velocity Threshold=1500
   )
   &CameraOffset(Camera Offset Amount=4)
  ]
  rendererClasses [&NiagaraSpriteRendererProperties, &NiagaraMeshRendererProperties]
  renderers [
   {rendererClass=&NiagaraSpriteRendererProperties}
   {rendererClass=&NiagaraMeshRendererProperties, rendererIndex=1}
  ]
 }
 FN_AnimTrail_blue {
  emitterSpawnScript [&FN_RibbonIDOffsetUtility()]
  emitterUpdateScript [
   &EmitterState()
   &FN_SpawnAnimTrail(AnimTrailWidthMode=::FromFirst)
  ]
  particleSpawnScript [
   &FN_InitializeAnimTrailBoneIndecies(First Bone or Socket Index=2, Second Bone or Socket Index=2)
   &FN_InitializeAnimTrailBoneIndecies(AnimTrail Selection=1, Second Bone or Socket Index=2) {enabled=false, module=FN_InitializeAnimTrailBoneIndecies001}
   &FN_InitializeAnimTrail(Color=<rgba 0 0.5333 1 1>, Lifetime=0.15, SkeletalMesh=$User.Swing_SK_Mesh)
   &AddVelocity(Velocity=<0 400 0>) {enabled=false}
  ]
  particleUpdateScript [
   &SolveForcesAndVelocity(Clamp Velocity=false, Limit Acceleration=false)
   &ScaleRibbonWidth(Ribbon Width Scale=&FloatFromCurve(
     FloatCurve=DataInterface() {propertyValues {Curve {keys [
         {arriveTangent=9.0725, interpMode=RCIM_Cubic, leaveTangent=9.0725, tangentMode=RCTM_User, time=0, value=0}
         {arriveTangent=-2.5249, interpMode=RCIM_Cubic, leaveTangent=-2.5249, tangentMode=RCTM_User, time=0.1408, value=0.9847}
         {arriveTangent=-0.0405, interpMode=RCIM_Cubic, leaveTangent=-0.0405, tangentMode=RCTM_User, time=1, value=0}
        ]}}}
     Scale Curve=15
    ))
   &ScaleColor(
    Scale Alpha=&FloatFromCurve(FloatCurve=DataInterface() {propertyValues {Curve {keys [
         {interpMode=RCIM_Cubic, tangentMode=RCTM_User, time=0.3, value=0}
         {interpMode=RCIM_Cubic, tangentMode=RCTM_User, time=0.5, value=1}
         {tangentMode=RCTM_User, time=1, value=0}
        ]}}})
    Scale RGB=&MakeVectorFromLinearColorRGB(Color=&ColorFromCurve(ColorCurve=DataInterface() {propertyValues {
        AlphaCurve {keys [<at 0.132 1 RCIM_Cubic>]}
        BlueCurve {keys [{arriveTangent=-34.2206, interpMode=RCIM_Cubic, leaveTangent=-34.2206, time=0, value=20}, <at 0.5494 1.2>, <at 1 1.2 RCIM_Cubic>]}
        GreenCurve {keys [{arriveTangent=-34.2206, interpMode=RCIM_Cubic, leaveTangent=-34.2206, time=0, value=20}, <at 0.5494 1.2>, <at 1 1.2 RCIM_Cubic>]}
        RedCurve {keys [{arriveTangent=-34.2206, interpMode=RCIM_Cubic, leaveTangent=-34.2206, time=0, value=20}, <at 0.5494 1.2>, <at 1 1.2 RCIM_Cubic>]}
       }}))
    ScaleA=true
    ScaleRGB=false
   ) {module=ScaleColor002}
   &FN_UpdateAnimTrail(SkeletalMesh=$User.Swing_SK_Mesh) {module=FN_UpdateAnimTrail001}
   /SetVariables(Particles.RibbonU0Override=$Particles.NormalizedAge) {enabled=false, isSetParameters=true, module=SetVariables_#1}
   &FN_RibbonMaterial(Color=$Particles.Color, Scale Alpha=0.8) {enabled=false}
   &CameraOffset(Camera Offset Amount=10)
  ]
  rendererClasses [&NiagaraRibbonRendererProperties]
  renderers [{rendererClass=&NiagaraRibbonRendererProperties}]
 }
 FN_AnimTrail_blue001 {
  emitterSpawnScript [&FN_RibbonIDOffsetUtility()]
  emitterUpdateScript [
   &EmitterState()
   &FN_SpawnAnimTrail(AnimTrailWidthMode=::FromFirst, Spawn Interval=0)
  ]
  particleSpawnScript [
   &FN_InitializeAnimTrailBoneIndecies(First Bone or Socket Index=2, Second Bone or Socket Index=2)
   &FN_InitializeAnimTrailBoneIndecies(AnimTrail Selection=1, Second Bone or Socket Index=2) {enabled=false, module=FN_InitializeAnimTrailBoneIndecies001}
   &FN_InitializeAnimTrail(AnimTrailWidth=4, Color=<rgba 0.1651 0.6514 1 1>, Lifetime=0.15, SkeletalMesh=$User.Swing_SK_Mesh)
   &AddVelocity(Velocity=<0 400 0>) {enabled=false}
  ]
  particleUpdateScript [
   &SolveForcesAndVelocity(Clamp Velocity=false, Limit Acceleration=false)
   &ScaleRibbonWidth(Ribbon Width Scale=&FloatFromCurve(
     FloatCurve=DataInterface() {propertyValues {Curve {keys [
         {arriveTangent=9.0725, interpMode=RCIM_Cubic, leaveTangent=9.0725, tangentMode=RCTM_User, time=0, value=0}
         {arriveTangent=-2.5249, interpMode=RCIM_Cubic, leaveTangent=-2.5249, tangentMode=RCTM_User, time=0.02, value=0.9847}
         {arriveTangent=-0.0405, interpMode=RCIM_Cubic, leaveTangent=-0.0405, tangentMode=RCTM_User, time=1, value=0}
        ]}}}
     Scale Curve=40
    ))
   &ScaleColor(
    Scale Alpha=&FloatFromCurve(
     FloatCurve=DataInterface() {propertyValues {Curve {keys [
         {interpMode=RCIM_Cubic, tangentMode=RCTM_User, time=0, value=0}
         {interpMode=RCIM_Cubic, tangentMode=RCTM_User, time=0.2, value=1}
         {tangentMode=RCTM_User, time=1, value=0}
        ]}}}
     Scale Curve=0.2
    )
    Scale RGB=&MakeVectorFromLinearColorRGB(Color=&ColorFromCurve(ColorCurve=DataInterface() {propertyValues {
        AlphaCurve {keys [<at 0.132 1 RCIM_Cubic>]}
        BlueCurve {keys [{arriveTangent=-34.2206, interpMode=RCIM_Cubic, leaveTangent=-34.2206, time=0, value=20}, <at 0.5494 1.2>, <at 1 1.2 RCIM_Cubic>]}
        GreenCurve {keys [{arriveTangent=-34.2206, interpMode=RCIM_Cubic, leaveTangent=-34.2206, time=0, value=20}, <at 0.5494 1.2>, <at 1 1.2 RCIM_Cubic>]}
        RedCurve {keys [{arriveTangent=-34.2206, interpMode=RCIM_Cubic, leaveTangent=-34.2206, time=0, value=20}, <at 0.5494 1.2>, <at 1 1.2 RCIM_Cubic>]}
       }}))
    ScaleA=true
    ScaleRGB=false
   ) {module=ScaleColor002}
   &FN_UpdateAnimTrail(SkeletalMesh=$User.Swing_SK_Mesh) {module=FN_UpdateAnimTrail001}
   /SetVariables(Particles.RibbonFacing=<0 0 0>) {enabled=false, isSetParameters=true, module=SetVariables_#2}
   /SetVariables(Particles.RibbonU0Override=$Particles.NormalizedAge) {enabled=false, isSetParameters=true, module=SetVariables_#1}
   &FN_RibbonMaterial(Color=$Particles.Color, Scale Alpha=0.8) {enabled=false}
   &DynamicMaterialParameters(Module.Param0WriteEnabled=true, Module.Param1WriteEnabled=true, Module.Param2WriteEnabled=true, Module.Param3WriteEnabled=true) {enabled=false}
  ]
  rendererClasses [&NiagaraRibbonRendererProperties]
  renderers [{rendererClass=&NiagaraRibbonRendererProperties}]
 }
 FN_AnimTrail_blue002 {
  emitterSpawnScript [&FN_RibbonIDOffsetUtility()]
  emitterUpdateScript [
   &EmitterState()
   &FN_SpawnAnimTrail(AnimTrailWidthMode=::FromFirst)
  ]
  particleSpawnScript [
   &FN_InitializeAnimTrailBoneIndecies(First Bone or Socket Index=2, Second Bone or Socket Index=2)
   &FN_InitializeAnimTrailBoneIndecies(AnimTrail Selection=1, Second Bone or Socket Index=2) {enabled=false, module=FN_InitializeAnimTrailBoneIndecies001}
   &FN_InitializeAnimTrail(AnimTrailWidth=3, Color=<rgba 0.1651 0.6514 1 1>, Lifetime=0.15, SkeletalMesh=$User.Swing_SK_Mesh)
   &AddVelocity(Velocity=<0 400 0>) {enabled=false}
  ]
  particleUpdateScript [
   &SolveForcesAndVelocity(Clamp Velocity=false, Limit Acceleration=false)
   &ScaleRibbonWidth(Ribbon Width Scale=&FloatFromCurve(
     FloatCurve=DataInterface() {propertyValues {Curve {keys [
         {arriveTangent=9.0725, interpMode=RCIM_Cubic, leaveTangent=9.0725, tangentMode=RCTM_User, time=0, value=0}
         {arriveTangent=-2.5249, interpMode=RCIM_Cubic, leaveTangent=-2.5249, tangentMode=RCTM_User, time=0.02, value=0.9847}
         {arriveTangent=-0.0405, interpMode=RCIM_Cubic, leaveTangent=-0.0405, tangentMode=RCTM_User, time=1, value=0}
        ]}}}
     Scale Curve=70
    ))
   &ScaleColor(
    Scale Alpha=&FloatFromCurve(
     FloatCurve=DataInterface() {propertyValues {Curve {keys [
         {interpMode=RCIM_Cubic, tangentMode=RCTM_User, time=0.2, value=0}
         {interpMode=RCIM_Cubic, tangentMode=RCTM_User, time=0.4, value=1}
         {tangentMode=RCTM_User, time=1, value=0}
        ]}}}
     Scale Curve=0.5
    )
    Scale RGB=&MakeVectorFromLinearColorRGB(Color=&ColorFromCurve(ColorCurve=DataInterface() {propertyValues {
        AlphaCurve {keys [<at 0.132 1 RCIM_Cubic>]}
        BlueCurve {keys [{arriveTangent=-34.2206, interpMode=RCIM_Cubic, leaveTangent=-34.2206, time=0, value=20}, <at 0.5494 1.2>, <at 1 1.2 RCIM_Cubic>]}
        GreenCurve {keys [{arriveTangent=-34.2206, interpMode=RCIM_Cubic, leaveTangent=-34.2206, time=0, value=20}, <at 0.5494 1.2>, <at 1 1.2 RCIM_Cubic>]}
        RedCurve {keys [{arriveTangent=-34.2206, interpMode=RCIM_Cubic, leaveTangent=-34.2206, time=0, value=20}, <at 0.5494 1.2>, <at 1 1.2 RCIM_Cubic>]}
       }}))
    ScaleA=true
    ScaleRGB=false
   ) {module=ScaleColor002}
   &FN_UpdateAnimTrail(SkeletalMesh=$User.Swing_SK_Mesh) {module=FN_UpdateAnimTrail001}
   /SetVariables(Particles.RibbonFacing=<0 0 0>) {enabled=false, isSetParameters=true, module=SetVariables_#2}
   /SetVariables(Particles.RibbonU0Override=$Particles.NormalizedAge) {enabled=false, isSetParameters=true, module=SetVariables_#1}
   &FN_RibbonMaterial(Color=$Particles.Color, Scale Alpha=0.8) {enabled=false}
  ]
  rendererClasses [&NiagaraRibbonRendererProperties]
  renderers [{rendererClass=&NiagaraRibbonRendererProperties}]
 }
 FN_AnimTrail_gray {
  emitterSpawnScript [&FN_RibbonIDOffsetUtility() {enabled=false}]
  emitterUpdateScript [
   &EmitterState()
   &FN_SpawnAnimTrail(Spawn Time=0.2)
  ]
  particleSpawnScript [
   &FN_InitializeAnimTrailBoneIndecies(First Bone or Socket Index=2, Second Bone or Socket Index=2)
   &FN_InitializeAnimTrailBoneIndecies(AnimTrail Selection=1, First Bone or Socket Index=1, Second Bone or Socket Index=2) {enabled=false, module=FN_InitializeAnimTrailBoneIndecies001}
   &FN_InitializeAnimTrail(AnimTrailWidth=30, Color=<rgba 0 0 0 4>, Lifetime=0.15, SkeletalMesh=$User.Swing_SK_Mesh)
   &AddVelocity(Velocity=<0 400 0>) {enabled=false}
  ]
  particleUpdateScript [
   &SolveForcesAndVelocity(Clamp Velocity=false, Limit Acceleration=false)
   &ScaleRibbonWidth(Ribbon Width Scale=&FloatFromCurve(FloatCurve=DataInterface() {propertyValues {Curve {keys [
         {interpMode=RCIM_Cubic, tangentMode=RCTM_User, time=0, value=0}
         {interpMode=RCIM_Cubic, tangentMode=RCTM_User, time=0.5, value=1}
         {tangentMode=RCTM_User, time=1, value=0}
        ]}}})) {enabled=false}
   &ScaleRibbonWidth(Ribbon Width Scale=&FloatFromCurve(
     FloatCurve=DataInterface() {propertyValues {Curve {keys [{tangentMode=RCTM_User, time=0, value=1}, {tangentMode=RCTM_User, time=1, value=0}]}}}
     Scale Curve=5
    )) {module=ScaleRibbonWidth001}
   &ScaleColor(
    Scale Alpha=&FloatFromCurve(
     FloatCurve=DataInterface() {propertyValues {Curve {keys [
         {interpMode=RCIM_Cubic, tangentMode=RCTM_User, time=0.4, value=0}
         {interpMode=RCIM_Cubic, tangentMode=RCTM_User, time=0.5, value=1}
         {tangentMode=RCTM_User, time=1, value=0}
        ]}}}
     Scale Curve=0.1
    )
    Scale RGB=&MakeVectorFromLinearColorRGB(Color=&ColorFromCurve(ColorCurve=DataInterface() {propertyValues {
        AlphaCurve {keys [<at 0 0>,<at 0.2586 1 RCIM_Cubic>]}
        BlueCurve {keys [<at 0 0.2 RCIM_Cubic>,<at 0.132 0.1917>,<at 0.3563 0.3>,<at 0.5534 0.1 RCIM_Cubic>]}
        GreenCurve {keys [<at 0 0.2 RCIM_Cubic>,<at 0.132 0.18>,<at 0.3563 0.3>,<at 0.5534 0.0933 RCIM_Cubic>]}
        RedCurve {keys [<at 0 0.2 RCIM_Cubic>,<at 0.132 0.2>,<at 0.3563 0.3>,<at 0.5534 0.09 RCIM_Cubic>]}
       }}))
    ScaleA=true
    ScaleRGB=true
   ) {module=ScaleColor002}
   &FN_UpdateAnimTrail(SkeletalMesh=$User.Swing_SK_Mesh) {module=FN_UpdateAnimTrail001}
   /SetVariables(Particles.RibbonU0Override=$Particles.NormalizedAge) {enabled=false, isSetParameters=true, module=SetVariables_#1}
   &FN_RibbonMaterial(Color=$Particles.Color, Opacity Pan U Speed=1, Scale Alpha=0.8) {enabled=false}
  ]
  rendererClasses [&NiagaraRibbonRendererProperties]
  renderers [{rendererClass=&NiagaraRibbonRendererProperties}]
 }
}
summary {
 systemName=NS_Pickaxe_Exosuit_Female_Swing
 emitters [
  {emitterName=FN_AnimTrail_blue, rendererClasses [&NiagaraRibbonRendererProperties]}
  {emitterName=FN_AnimTrail_gray, rendererClasses [&NiagaraRibbonRendererProperties]}
  {emitterName=FN_AnimTrail_blue001, rendererClasses [&NiagaraRibbonRendererProperties]}
  {emitterName=FN_AnimTrail_blue002, rendererClasses [&NiagaraRibbonRendererProperties]}
  {
   emitterName=DirectionalBurst
   rendererClasses [&NiagaraSpriteRendererProperties, &NiagaraMeshRendererProperties]
  }
 ]
 userVariables [
  User.RibbonIDOffset {type=NiagaraInt32}
  User.SkeletalMesh {
   type=NiagaraDataInterfaceSkeletalMesh
   defaultValue {
    struct=NiagaraExt_VariableValue_DataInterface
    value {dataInterface=&self:NiagaraDataInterfaceSkeletalMesh_4}
   }
  }
  User.Smoke {type=NiagaraFloat, defaultValue=0, description=""}
  User.Swing_SK_Mesh {
   type=NiagaraDataInterfaceSkeletalMesh
   defaultValue {
    struct=NiagaraExt_VariableValue_DataInterface
    value {dataInterface=&self:NiagaraDataInterfaceSkeletalMesh_0}
   }
  }
  User.TrailLifeTime {type=NiagaraFloat, defaultValue=0, description=""}
 ]
}
