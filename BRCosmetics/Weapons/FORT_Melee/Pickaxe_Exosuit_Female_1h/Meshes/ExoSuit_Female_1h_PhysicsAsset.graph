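# Physics asset for the one-handed female ExoSuit pickaxe, a cosmetic melee weapon mesh.
# It holds one body on Pickaxe_ExoSuit_Female_1h_MESH_LOD0, shaped as a single tilted, offset capsule.
# There are no constraints, so the pickaxe is represented as one piece with no articulated parts.
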PhysicsAsset /BRCosmetics/Weapons/FORT_Melee/Pickaxe_Exosuit_Female_1h/Meshes/ExoSuit_Female_1h_PhysicsAsset
bodies {Pickaxe_ExoSuit_Female_1h_MESH_LOD0 [{shapeType=Capsule, length=44.2924, radius=10.2689, center=<2.7893 0.0013 26.3046>, rotation=<rot -0.0057 -90 7.3533>}]}
constraints []
